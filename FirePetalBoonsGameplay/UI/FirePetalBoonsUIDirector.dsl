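; Dynamic UI director that shows and hides UI scenes for the Hot Hands, Haunted Helpers and Void Run boons.
; At BeginPlay it starts three tag watchers on the owning player controller and binds to view target changes, which start watchers on each new view target.
; Each watcher runs async gameplay tag add and remove waits side by side, adding or removing its matching scene.
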(event EventBeginPlay
  (bind _asfort_player_controller (Utilities|Casting|CastToFortPlayerController (DynamicUIDirector|GetOwningLocalPlayerController)))
  (EventDispatchers|BindEventtoOnViewTargetChanged _asfort_player_controller (delegate Custom|OnViewTargetChanged))
  (CallFunction|WaitforPlayerHotHandsTag _asfort_player_controller)
  (CallFunction|WaitforPlayerHauntedHelpers _asfort_player_controller)
  (CallFunction|WaitforPlayerVoidRun _asfort_player_controller))

(event Custom|OnViewTargetChanged
  (params (PlayerController :type Fort%20Player%20Controller%20Object%20Reference) (OldViewTarget :type Actor%20Object%20Reference) (NewViewTarget :type Actor%20Object%20Reference))
  (CallFunction|WaitforPlayerHotHandsTag NewViewTarget)
  (CallFunction|WaitforPlayerHauntedHelpers NewViewTarget)
  (CallFunction|WaitforPlayerVoidRun NewViewTarget))

(event Custom|WaitForPlayerHotHandsTag
  (params (PlayerActor :type Actor%20Object%20Reference))
  (bind _returnvalue (DynamicUIDirector|GetOwningLocalPlayerController))
  ;[1 1 1] Wait For Player Hot Hands Tag
  (Utilities|IsValid PlayerActor
    (:Is%20Valid
      (Utilities|IsValid _returnvalue
        (:Is%20Valid
          (Utilities|Casting|CastToFortPlayerPawnAthena (Pawn|GetControlledPawn _returnvalue))
          (Ability|Tasks|WaitGameplayTagAddToActor PlayerActor (Variables|Components|GetHotHandsTag_Event)
            (:then
              (Ability|Tasks|WaitGameplayTagRemoveFromActor PlayerActor (Variables|Components|GetHotHandsTag_Event)
                (:Removed
                  (DynamicUIDirector|RemoveScene "/FirePetalBoonsGameplay/UI/FirePetalBoonsUIScene.FirePetalBoonsUIScene"))))
            (:Added
              (DynamicUIDirector|AddScene "/FirePetalBoonsGameplay/UI/FirePetalBoonsUIScene.FirePetalBoonsUIScene"))))))))

(event Custom|WaitForPlayerHauntedHelpers
  (params (PlayerActor :type Actor%20Object%20Reference))
  (bind _returnvalue (DynamicUIDirector|GetOwningLocalPlayerController))
  ;[1 1 1] Wait For Player Haunted Helpers
  (Utilities|IsValid PlayerActor
    (:Is%20Valid
      (Utilities|IsValid _returnvalue
        (:Is%20Valid
          (Utilities|Casting|CastToFortPlayerPawnAthena (Pawn|GetControlledPawn _returnvalue))
          (Ability|Tasks|WaitGameplayTagAddToActor PlayerActor (Variables|Components|GetHauntedHelpersTag_Event)
            (:then
              (Ability|Tasks|WaitGameplayTagRemoveFromActor PlayerActor (Variables|Components|GetHauntedHelpersTag_Event)
                (:Removed
                  (DynamicUIDirector|RemoveScene "/FirePetalBoonsGameplay/UI/HauntedHandsBoonsUIScene.HauntedHandsBoonsUIScene"))))
            (:Added
              (DynamicUIDirector|AddScene "/FirePetalBoonsGameplay/UI/HauntedHandsBoonsUIScene.HauntedHandsBoonsUIScene"))))))))

(event Custom|WaitForPlayerVoidRun
  (params (PlayerActor :type Actor%20Object%20Reference))
  (bind _returnvalue (DynamicUIDirector|GetOwningLocalPlayerController))
  ;[1 1 1] Wait For Player VoidRun
  (Utilities|IsValid PlayerActor
    (:Is%20Valid
      (Utilities|IsValid _returnvalue
        (:Is%20Valid
          (Utilities|Casting|CastToFortPlayerPawnAthena (Pawn|GetControlledPawn _returnvalue))
          (Ability|Tasks|WaitGameplayTagAddToActor PlayerActor (Variables|Components|GetVoidRunTag_Event)
            (:then
              (Ability|Tasks|WaitGameplayTagRemoveFromActor PlayerActor (Variables|Components|GetVoidRunTag_Event)
                (:Removed
                  (DynamicUIDirector|RemoveScene "/FirePetalBoonsGameplay/UI/VoidRunBoonsUIScene.VoidRunBoonsUIScene"))))
            (:Added
              (DynamicUIDirector|AddScene "/FirePetalBoonsGameplay/UI/VoidRunBoonsUIScene.VoidRunBoonsUIScene"))))))))

(fn ConstructionScript)
