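# Translucent UI material for the Boons ammo counter VFX, drawing scrolling, two-color smoke that follows animated pip progress.
# Panned Perlin noise distorts the smoke texture's UVs, which blend between flat and radial coordinates through isRadial.
# The mask's red channel feeds opacity, green blends main to accent color, and blue slightly darkens it.
# isRadial also switches opacity between a mirrored linear-gradient layout and a radial one, with arc and horizontal progress masks defined.
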Material /FirePetalBoonsGameplay/UI/Materials/M_UI_BoonsAmmoCounter_VFx
graph {
 properties {bAutomaticallySetUsageInEditor=false, blendMode=BLEND_Translucent, materialDomain=MD_UI, usedShadingModels=Default Lit}
 nodes [
  MaterialFunctionCall() {materialFunction=&DebugTimeSine}
  n1 = Multiply(A=NamedRerouteUsage(), B=NamedRerouteUsage())
  n2 = MaterialFunctionCall(UVs=MaterialFunctionCall!(UV, Offset, Rotation, Scale, MirrorY) {materialFunction=/FortUILibrary/MaterialFunctions/Basics/MF_UI_Symmetry_2x1_DEPRECATED}.Result) {materialFunction=&MF_UI_LinearGradient}
  n3 = NamedRerouteUsage()
  n4 = NamedRerouteUsage()
  Multiply(
   A=LinearInterpolate(A=Multiply(A=@n1, B=@n2.Result), B=Multiply(A=@n3, B=@n4), Alpha=ScalarParameter() {parameterName=isRadial})
   B=ScalarParameter() {defaultValue=1, parameterName=Opacity}
  )
  n5 = ScalarParameter() {defaultValue=0.25, group=Radial, parameterName=MaskStartAngle}
  n6 = ScalarParameter() {defaultValue=0.25, group=Radial, parameterName=MaskEndAngle}
  n7 = ScalarParameter() {
   defaultValue=1
   desc="Property for animating progress value change (from ProgressStart to ProgressEnd)\r\n0 - 1"
   group=Progress
   parameterName=AnimatePipProgress
   sliderMax=1
   sortPriority=3
  }
  n8 = OneMinus(None=@n7)
  n9 = ScalarParameter() {defaultValue=1, group=Radial, parameterName=FillClockwise?}
  n10 = MaterialFunctionCall!(UV, "Arc Start Angle (0-1)"=OneMinus(None=@n6), "Arc End Angle (0-1)"=OneMinus(None=@n5), "Fill Ratio (0-1)"=@n8, Fill Clockwise=@n9) {materialFunction=/FortUILibrary/MaterialFunctions/Basics/MF_UI_RadialArc}
  NamedRerouteDeclaration(None=Min(
    A=MaterialFunctionCall("Arc Start Angle (0-1)"=@n5, "Arc End Angle (0-1)"=@n6, "Fill Ratio (0-1)"=@n8, Fill Clockwise=@n9) {materialFunction=&MF_UI_RadialArc}.Radial Fill
    B=Multiply(A=@n10.Radial Fill) {constB=1.5}
   )) {name=RadialMask, nodeColor=<rgba 0 0 0 1>}
  n11 = TextureSampleParameter2D(UVs=NamedRerouteUsage()) {
   desc=Should have two channels R is Main mask and G is Accent
   parameterName=Mask
   texture=/FirePetalBoonsGameplay/UI/Materials/Textures/T_UI_Boon_HauntedSmoke_128
  }
  NamedRerouteDeclaration(None=Add(A=@n11.R, B=@n11.R)) {name=Layers, nodeColor=<rgba 0 0 0 1>}
  n12 = TextureCoordinate()
  n13 = Constant()
  n14 = ScalarParameter() {
   desc="When progress value changes, it defines where change starts.\r\n0 - 1"
   group=Progress
   parameterName=PipProgressStart
   sliderMax=1
   sortPriority=1
  }
  n15 = ScalarParameter() {
   defaultValue=1
   desc="When progress value changes, it defines where change ends.\r\n0 - 1"
   group=Progress
   parameterName=PipProgressEnd
   sliderMax=1
   sortPriority=2
  }
  n16 = ScalarParameter() {
   defaultValue=1
   desc="Property for animating progress value change (from ProgressStart to ProgressEnd)\r\n0 - 1"
   group=Progress
   parameterName=AnimatePipProgress
   sliderMax=1
   sortPriority=3
  }
  n17 = Add(
   A=Multiply(A=MaterialFunctionCall(UVs=@n12, "Rotation Angle (0-1)"=@n13) {materialFunction=&CustomRotator}.Rotated Values) {constB=-1}
   B=NamedRerouteDeclaration(None=LinearInterpolate(A=@n14, B=@n15, Alpha=@n16)) {name=Progress_1, nodeColor=<rgba 0 0 0 1>}
  ) {constB=0.5}
  NamedRerouteDeclaration(None=Saturate(None=ComponentMask(None=@n17) {r=true})) {name=ProgressGraident_Horizontal, nodeColor=<rgba 0 0 0 1>}
  n18 = ScalarParameter() {defaultValue=0.4091, group=Distortion, parameterName=DistortionSpeedX}
  n19 = ScalarParameter() {group=Distortion, parameterName=DistortionSpeedY}
  n20 = TextureSampleParameter2D(UVs=Panner(Coordinate=TextureCoordinate(), Speed=AppendVector(A=@n18, B=@n19)) {speedX=0.4}) {group=Distortion, parameterName=DistortionNoise, texture=/FirePetalBoonsGameplay/UI/Materials/Textures/T_UI_PerlinNoise}
  n21 = AppendVector(A=@n20.G, B=@n20.G)
  n22 = ScalarParameter() {defaultValue=0.24, group=Distortion, parameterName=DistortionSize}
  n23 = TextureCoordinate()
  n24 = MaterialFunctionCall!(Vector or UVs=MaterialFunctionCall(UVs=@n23, Texture Scale=ScalarParameter() {defaultValue=0.7, group=Radial, parameterName=ScaleOfRadial}) {materialFunction=&ScaleUVsByCenter}.UVs) {materialFunction=/FortUILibrary/MaterialFunctions/Basics/MF_UI_CheaperVectorToRadial}
  n25 = ScalarParameter() {parameterName=isRadial}
  n26 = Panner(
   Coordinate=Add(A=Multiply(A=@n21, B=@n22) {constB=0.24}, B=LinearInterpolate(A=@n23, B=@n24.Radial Coordinates, Alpha=@n25))
   Speed=AppendVector(A=ScalarParameter() {defaultValue=1.1, parameterName=SpeedX}, B=ScalarParameter() {parameterName=SpeedY})
  ) {speedX=1.1}
  n27 = ComponentMask(None=@n26) {g=true}
  n28 = ScalarParameter() {defaultValue=-0.0853, parameterName=PlacementOnVAxis}
  NamedRerouteDeclaration(None=AppendVector(A=ComponentMask(None=@n26) {r=true}, B=Add(A=@n27, B=@n28))) {name=UV's, nodeColor=<rgba 0 0 0 1>}
  Subtract(
   A=LinearInterpolate(
    A=VectorParameter() {parameterName=MainColor, defaultValue=<rgba 0.1045 0.7622 0.2954 1>}.RGB
    B=VectorParameter() {parameterName=AccentColor, defaultValue=<rgba 1 1 1 1>}.RGB
    Alpha=@n11.G
   )
   B=Multiply(A=@n11.B) {constB=0.1}
  )
 ]
}
