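# Translucent UI material for a rounded Void Run indicator, drawing an animated noisy radial progress bar with a faint circle outline.
# Two samples of a distortion texture pan in opposite horizontal directions and combine into a mask that tints the bar's fill from purple to white.
# The boosted, saturated mask multiplies the radial bar function's opacity, then a Max merges it with the outline scaled by the Circle Opacity parameter.
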Material /FirePetalBoonsGameplay/UI/Materials/M_UI_VoidRun_Rounded
graph {
 properties {bAutomaticallySetUsageInEditor=false, blendMode=BLEND_Translucent, materialDomain=MD_UI, usedShadingModels=Default Lit}
 nodes [
  n1 = TextureSample(UVs=Panner(Coordinate=MaterialFunctionCall() {materialFunction=&GetUserInterfaceUV}.Normalized UV) {speedX=-0.2}) {texture=/MasamuneGameplay/UI/Materials/T_UI_VoidRun_Distortion}
  n2 = MaterialFunctionCall() {materialFunction=&GetUserInterfaceUV}
  n3 = TextureSample(UVs=Panner(Coordinate=Multiply(A=@n2.Normalized UV) {constB=-0.8}) {speedX=0.25, speedY=0.05}) {texture=&T_UI_VoidRun_Distortion}
  n4 = Min(A=Multiply(A=@n1.RGB, B=@n3.RGB), B=Saturate(None=Add(A=@n1.RGB, B=@n3.RGB)))
  n5 = LinearInterpolate(A=Constant3Vector() {constant=<rgba 0.4011 0.184 1 1>}, B=Constant3Vector() {constant=<rgba 1 1 1 0>}, Alpha=@n4)
  n6 = Multiply(A=@n4) {constB=10}
  n7 = Constant() {r=0.9}
  Max(
   A=Multiply(
    A=MaterialFunctionCall!(FillColor=@n5, BGColor) {materialFunction=/Game/Athena/HUD/Bacchus/Material/MF_RadialProgBar}.Opacity
    B=Saturate(None=@n6)
   )
   B=Multiply(
    A=ScalarParameter() {defaultValue=0.2, parameterName=Circle Opacity}
    B=MaterialFunctionCall(Size=@n7) {materialFunction=&MF_UI_Circle}.Outline
   ) {constB=0.2}
  )
 ]
}
