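# Translucent UI material for the Void Run icon, combining a tactical sprint icon texture with animated distortion and a constant purple-blue color.
# A panner scrolls the Void Run distortion texture horizontally over normalized UI UVs.
# The icon's red channel is doubled and multiplied by the offset red-green distortion channels, so the scrolling texture modulates the icon shape's intensity rather than warping its UVs.
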Material /FirePetalBoonsGameplay/UI/Materials/M_UI_VoidRunIcon
graph {
 properties {bAutomaticallySetUsageInEditor=false, blendMode=BLEND_Translucent, materialDomain=MD_UI, usedShadingModels=Default Lit}
 nodes [
  Constant3Vector() {constant=<rgba 0.4011 0.184 1 1>}
  n1 = TextureSample() {samplerType=SAMPLERTYPE_LinearColor, texture=/TacticalSprintUI/UI/Textures/T_UI_TacSprint_64x}
  n2 = TextureSample(UVs=Panner(Coordinate=MaterialFunctionCall() {materialFunction=&GetUserInterfaceUV}.Normalized UV) {speedX=-0.2}) {texture=/MasamuneGameplay/UI/Materials/T_UI_VoidRun_Distortion}
  Multiply(A=ComponentMask(None=Add(A=@n1.R, B=@n1.R)) {r=true}, B=Add(A=ComponentMask(None=@n2.RGB) {g=true, r=true}) {constB=0.5})
 ]
}
