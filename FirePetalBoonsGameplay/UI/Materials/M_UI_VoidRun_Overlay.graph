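# UI material for a translucent Void Run screen overlay that draws animated purple-to-white distortion.
# Two samples of one distortion texture pan at different speeds, opposite in X, and combine as the minimum of their product and saturated sum.
# A procedural box distance field from UVs, plus a thin smoothstepped band where it nears zero, masks that noise.
# The result drives the color lerp and, boosted and saturated, a second mask, likely opacity.
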Material /FirePetalBoonsGameplay/UI/Materials/M_UI_VoidRun_Overlay
graph {
 properties {blendMode=BLEND_Translucent, materialDomain=MD_UI, pixelDepthOffsetMode=PDOM_Legacy, usedShadingModels=Default Lit}
 nodes [
  n1 = TextureSample(UVs=Panner(Coordinate=MaterialFunctionCall() {materialFunction=&GetUserInterfaceUV}.Normalized UV) {speedX=-0.2}) {texture=/MasamuneGameplay/UI/Materials/T_UI_VoidRun_Distortion}
  n2 = MaterialFunctionCall() {materialFunction=&GetUserInterfaceUV}
  n3 = TextureSample(UVs=Panner(Coordinate=Multiply(A=@n2.Normalized UV) {constB=-0.8}) {speedX=0.25, speedY=0.05}) {texture=&T_UI_VoidRun_Distortion}
  n4 = Add(A=@n1.RGB, B=@n3.RGB)
  n5 = NamedRerouteUsage()
  n6 = SmoothStep(Min=Constant(), Max=Constant() {r=0.01}, Value=Subtract(A=Abs(None=@n5), B=Constant() {r=0.05}))
  n7 = Multiply(A=Min(A=Multiply(A=@n1.RGB, B=@n3.RGB), B=Saturate(None=@n4)), B=Add(A=NamedRerouteUsage(), B=OneMinus(None=@n6)))
  LinearInterpolate(A=Constant3Vector() {constant=<rgba 0.4011 0.184 1 1>}, B=Constant3Vector() {constant=<rgba 1 1 1 0>}, Alpha=@n7)
  n8 = TextureCoordinate()
  n9 = Subtract(A=Abs(None=ConstantBiasScale(None=@n8) {bias=-0.5, scale=2}), B=Constant2Vector() {g=1, r=1})
  n10 = Max(A=@n9) {constB=0}
  n11 = MaterialFunctionCall(Float2=@n9) {materialFunction=&BreakOutFloat2Components}
  n12 = Min(A=Max(A=@n11.R, B=@n11.R) {constB=0}) {constB=0}
  NamedRerouteDeclaration(None=Length(None=Add(A=@n10, B=@n12)) {is=class MaterialExpressionLength}) {name=Box, nodeColor=<rgba 0 0 0 1>}
  Saturate(None=Multiply(A=@n7) {constB=10})
 ]
}
