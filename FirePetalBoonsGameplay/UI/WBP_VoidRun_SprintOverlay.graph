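# Void Run sprint overlay widget, extending the Tactical Sprint layout with a material-driven overlay and icon for the constant energy regen effect.
# On construct it loops a glow material animation, nudges the overlay for mobile touch, and starts a timer on an empty duration-update event.
# On mobile touch, async gameplay tag listeners show or collapse the overlay.
# A helper finds the longest remaining effect duration but is never called here.
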WidgetBlueprint /FirePetalBoonsGameplay/UI/WBP_VoidRun_SprintOverlay
animations [Anim_EnergyGlowAnimation {
  binding_count=1
  length_seconds=5
  object=&self:Anim_EnergyGlowAnimation
  sequence {bindings [VoidRunOverlay {
     id=#1
     possessed_class=Image
     tracks [{
       type=MovieSceneWidgetMaterialTrack
       sections [{
         type=MovieSceneParameterSection
         end_frame=100
         start_frame=0
         channels [
          GlowMoving {type=MovieSceneScriptingFloatChannel, keys [[0,0],[20,1.8],[21,0],[35,1.8]]}
          ShineOpacity {type=MovieSceneScriptingFloatChannel, keys [[0,1],[19,1],[20,0],[21,0],[22,1]]}
         ]
        }]
      }]
    }]}
 }]
graphs [
 EventGraph {nodes [
   PlayGlowAnimation() {is=title K2Node_CustomEvent}
   > PlayAnimation(InAnimation=Get Anim_EnergyGlowAnimation() -> (Anim_EnergyGlowAnimation:Widget Animation Object Reference), NumLoopsToPlay=0)
   Event Construct()
   > Parent: Construct()
   > PlayGlowAnimation() {is=title K2Node_CallFunction}
   > Set Timer by Event(Delegate=OnPeriodicDurationUpdate() {is=title K2Node_CustomEvent}, Time=0.25, bLooping=true)
   > OnPeriodicDurationUpdate() {is=title K2Node_CallFunction}
   n1 = IsMobileGame()
   n2 = IsUsingTouch(PlayerController=GetOwningPlayer())
   > SetRenderTranslation(
    self=Get VoidRunOverlay() -> (VoidRunOverlay:Image Object Reference)
    Translation=Select(Option 0:Vector 2D Structure, Option 1:Vector 2D Structure=(X=1,Y=0), Index=AND Boolean(A=@n2, B=@n1)) -> (ReturnValue:Vector 2D Structure)
   )
   Event OnInitialized()
   n3 = GetOwningPlayer()
   > Branch(Condition=AND Boolean(A=IsUsingTouch(PlayerController=@n3), B=IsMobileGame()))
   n4 = GetOwningPlayer()
   > n5 = WaitGameplayTagAddToActor!(
    execute:Exec
    TargetActor:Actor Object Reference=@n4
    Tag:Gameplay Tag Structure=Get ConstantRegenTag_Active() -> (ConstantRegenTag_Active:Gameplay Tag Structure)
    OnlyTriggerOnce:bool=false
   ) -> (then:Exec, AsyncTaskProxy:Ability Async Wait Gameplay Tag Added Object Reference, Added:Exec) {is=title K2Node_AsyncAction}
   > n6 = WaitGameplayTagRemoveFromActor!(execute:Exec, TargetActor:Actor Object Reference=@n4, Tag:Gameplay Tag Structure=Get ConstantRegenTag_Active(), OnlyTriggerOnce:bool=false) -> (then:Exec, AsyncTaskProxy:Ability Async Wait Gameplay Tag Removed Object Reference, Removed:Exec) {is=title K2Node_AsyncAction}
   SetVisibility(execute=@n6.Removed, self=Get VoidRunOverlay(), InVisibility=Collapsed)
   SetVisibility(execute=@n5.Added, self=Get VoidRunOverlay(), InVisibility=SelfHitTestInvisible)
  ]}
 FindLongestActiveTimeRemaining {nodes [
   FindLongestActiveTimeRemaining() {is=title K2Node_FunctionEntry}
   > n1 = GetActiveEffectsWithAllTags!(
    execute
    self:Ability System Component Object Reference=GetAbilitySystemComponent!(Actor:Actor Object Reference=GetOwningPlayer()) -> (ReturnValue:Ability System Component Object Reference)
    Tags:Gameplay Tag Container Structure=MakeGameplayTagContainerFromTag!(SingleTag:Gameplay Tag Structure=Get ConstantRegenTag_Effect() -> (ConstantRegenTag_Effect:Gameplay Tag Structure)) -> (ReturnValue:Gameplay Tag Container Structure)
   ) -> (then, ReturnValue:Array of Active Gameplay Effect Handle Structures)
   > n2 = For Each Loop(Array:Array of Active Gameplay Effect Handle Structures=@n1) -> (Array Element:Active Gameplay Effect Handle Structure)
   Return Node!(execute=@n2.Completed, TimeRemaining:double=Get LongestTimeRemaining() -> (LongestTimeRemaining:double))
   > n3 = GetActiveGameplayEffectRemainingDuration!(execute, ActiveHandle:Active Gameplay Effect Handle Structure=@n2) -> (then, ReturnValue:float)
   > Branch(Condition="float > float"(A=@n3, B=Get LongestTimeRemaining()))
   > Set LongestTimeRemaining(LongestTimeRemaining:double=@n3) -> (Output_Get:double)
  ]}
]
variables [
 ConstantRegenTag_Active {type=GameplayTag, default {tagName=Granted.Athena.Energy.ConstantRegen}}
 ConstantRegenTag_Effect {type=GameplayTag, default {tagName=Asset.Athena.Energy.ConstantRegen}}
 RemainingTimeTextFormat {type=string, default="{seconds}s"}
]
widgets {
 info {
  inheritedWidgetCount=0
  namedSlotCount=2
  parentClass=/TacticalSprintUI/UI/WBP_TacSprint_Layout.WBP_TacSprint_Layout_C
  rootWidgetClass=null
  widgetCount=2
 }
 widgets [
  VoidRunOverlay {
   widgetClassPath=&Image
   bIsVariable=true
   widget=VoidRunOverlay
   properties {
    visibility=SelfHitTestInvisible
    brush {
     bIsDynamicallyLoaded=false
     drawAs=Box
     imageType=FullColor
     mirroring=NoMirror
     resourceName=None
     resourceObject=/FirePetalBoonsGameplay/UI/Materials/M_UI_VoidRun_Overlay
     tiling=NoTile
     imageSize=<32 32>
     margin=<ltrb 0.1 0.1 0.1 0.1>
     outlineSettings {
      bUseBrushTransparency=false
      roundingType=HalfHeightRadius
      width=0
      color {colorUseRule=UseColor_Specified, specifiedColor=<rgba 0 0 0 0>}
      cornerRadii=<0 0 0 0>
     }
     tintColor {colorUseRule=UseColor_Specified, specifiedColor=<rgba 1 1 1 1>}
     uVRegion {bIsValid=false, max=<0 0>, min=<0 0>}
    }
    renderTransform {angle=0, scale=<1 1>, shear=<0 0>, translation=<0 -4>}
   }
  }
  SprintIcon {
   widgetClassPath=&Image
   bIsVariable=true
   widget=SprintIcon
   properties {
    brush {
     bIsDynamicallyLoaded=false
     drawAs=Image
     imageType=FullColor
     mirroring=NoMirror
     resourceName=None
     resourceObject=/FirePetalBoonsGameplay/UI/Materials/M_UI_VoidRunIcon
     tiling=NoTile
     imageSize=<512 512>
     margin=<ltrb 0 0 0 0>
     outlineSettings {
      bUseBrushTransparency=false
      roundingType=HalfHeightRadius
      width=0
      color {colorUseRule=UseColor_Specified, specifiedColor=<rgba 0 0 0 0>}
      cornerRadii=<0 0 0 0>
     }
     tintColor {colorUseRule=UseColor_Specified, specifiedColor=<rgba 1 1 1 1>}
     uVRegion {bIsValid=false, max=<0 0>, min=<0 0>}
    }
    renderTransform {angle=0, scale=<0.75 0.75>, shear=<0 0>, translation=<6 1>}
   }
  }
 ]
}
